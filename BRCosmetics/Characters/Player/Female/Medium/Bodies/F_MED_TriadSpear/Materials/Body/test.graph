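# Test character body material for the TriadSpear cosmetic, adding audio-reactive emissive between the shared CharacterParent PreFX and PostFX functions.
# A frontend-loop audio function supplies amplitude and spectrum, exposed through named reroutes.
# Mask channels and a reroute signal drive a Ceil-quantized SmoothStep reveal of emissive detail over a base emissive texture.
# Cylindrical-UV sine distortion and time-pulse branches are left unconnected.
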Material /BRCosmetics/Characters/Player/Female/Medium/Bodies/F_MED_TriadSpear/Materials/Body/test
graph {
 properties {
  bAutomaticallySetUsageInEditor=false
  bUseMaterialAttributes=true
  bUsedWithMeshDeformer=true
  bUsedWithMorphTargets=true
  bUsedWithSkeletalMesh=true
  pixelDepthOffsetMode=PDOM_Legacy
  shadingModel=MSM_FromMaterialExpression
  usedShadingModels=Default Lit | Subsurface | Clear Coat | Subsurface Profile | Cloth
 }
 nodes [
  n1 = Constant() {r=1}
  n2 = Constant() {r=-7.1828}
  n3 = ComponentMask(None=Multiply(
    A=MaterialFunctionCall!(Tile=@n1, DistortionAdjustment=@n2) {materialFunction=&/MF_CylindricalProjUV}.UVs
    B=Constant() {r=5.274}
   )) {g=true}
  n4 = Saturate(None=MaterialFunctionCall(
    Input=Constant() {r=0.36}
    Input Low=ScalarParameter() {defaultValue=0.25, group=TopSide_V02_Layer Emissive_01 Reactive, parameterName=TopS Audio 01_Low_Remap, sortPriority=149}
    Input High=ScalarParameter() {defaultValue=1, group=TopSide_V02_Layer Emissive_01 Reactive, parameterName=TopS Audio 01_High_Remap, sortPriority=149}
    Target Low=Constant()
    Target High=Constant() {r=1}
   ) {materialFunction=&RemapValueRange}.Result)
  n5 = ScalarParameter() {defaultValue=0.8503, group=TopSide_V02_Layer Emissive_01 Reactive, parameterName=TopS Audio 01_Falloff, sortPriority=149}
  n6 = Multiply(A=Multiply(A=Sine(None=@n3), B=Power(Base=@n4, Exp=@n5)), B=Constant() {r=-0.1131})
  n7 = Constant()
  n8 = ScalarParameter() {defaultValue=-0.5, parameterName=Param_1_min}
  n9 = ScalarParameter() {defaultValue=0.5, parameterName=Param_3_Max}
  n10 = MaterialFunctionCall!(InAmplitude, InSpectrum, CustomFakeSpectrum, CustomFakeAmplitude) {materialFunction=/BRCosmetics/Global/MaterialFunctions/AudioReactivity/MF_AudioReactivity_FrontendLoop}
  n11 = Power(Base=@n10.Amplitude, Exp=ScalarParameter() {defaultValue=3, parameterName=Param_5}) {constExponent=4}
  Add(A=TextureCoordinate(), B=Multiply(A=Add(A=@n6, B=@n7), B=LinearInterpolate(A=@n8, B=@n9, Alpha=@n11) {constA=-0.05, constB=0.05}))
  ComponentMask(None=@n10.Spectrum) {r=true}
  n12 = MaterialFunctionCall() {materialFunction=&MF_CharacterParent_2023_PreFX}
  n13 = TextureSample() {texture=&/T_TriadSpear_Body_E2}
  n14 = TextureSample() {
   samplerType=SAMPLERTYPE_Masks
   texture=&/T_audio_Body_TAMask
  }
  n15 = Add(
   A=TextureSample() {texture=&/T_TriadSpear_Body_E_NOLine}.RGB
   B=Multiply(A=@n13.RGB, B=LinearInterpolate(Alpha=@n14.G) {constA=0.002, constB=0.02}) {constB=0.002}
  )
  n16 = NamedRerouteUsage()
  n17 = Power(Base=Multiply(A=@n14.G) {constB=1.5})
  n18 = LinearInterpolate(B=ScalarParameter() {defaultValue=0.5, parameterName=Param_6}, Alpha=Saturate(None=@n17)) {constA=0.1, constB=0.3}
  n19 = LinearInterpolate(
   A=ScalarParameter() {defaultValue=1.3, parameterName=Param_2}
   B=ScalarParameter() {parameterName=Param_3}
   Alpha=Saturate(None=Add(A=@n16, B=@n18) {constB=0.05})
  ) {constA=5, constB=0}
  n20 = OneMinus(None=@n14.R)
  n21 = ScalarParameter() {defaultValue=2, parameterName=3}
  n22 = ScalarParameter() {defaultValue=50, parameterName=Param_4}
  n23 = Ceil(None=Multiply(A=Power(Base=@n20, Exp=@n21) {constExponent=3}, B=@n22) {constB=10})
  n24 = SmoothStep(Min=ScalarParameter(), Max=@n19, Value=Saturate(None=Divide(A=@n23, B=@n22) {constB=10}))
  n25 = Multiply(A=Saturate(None=Multiply(A=@n13.RGB, B=@n24)), B=LinearInterpolate(Alpha=@n16) {constA=0.5, constB=2})
  MaterialFunctionCall(
   Material=SetMaterialAttributes!(MaterialAttributes=GetMaterialAttributes(None=@n12.Result).MaterialAttributes, Base Color, Emissive Color=Add(A=@n15, B=@n25)) {attributeSetTypes [#1, #2]}
   M=@n12.M
   Pre Skinned Local Position=@n12.M
   Distance Blend=@n12.M
   Fresnel=@n12.M
  ) {materialFunction=&MF_CharacterParent_2023_PostFX}
  Multiply(A=@n10.Amplitude) {constB=2}
  n26 = Multiply(A=Time()) {constB=0.1}
  Multiply(A=@n17, B=Sine(None=Add(A=@n17, B=@n26)))
  NamedRerouteDeclaration(None=@n11) {name=In_Ampltiude, nodeColor=<rgba 1 0.2812 0 1>}
  NamedRerouteDeclaration(None=@n10.Spectrum) {name=In_Ampltiude_1, nodeColor=<rgba 1 0.2812 0 1>}
  Saturate(None=@n19)
  Saturate(None=@n11)
  ScalarParameter() {defaultValue=5, parameterName=Param_1}
 ]
}
